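; Player camera modifier for the DentureSay air attack that scales camera rotation input while it is active.
; When added to the camera, it reads the Hotfix_RotationMultiplier scalable float once and caches the value.
; View rotation processing passes the current rotation through, multiplies the rotation delta by the cached multiplier, and returns true, which stops lower-priority modifiers from adjusting rotation further.
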(event EventBlueprintAddedToCamera
  (params (Camera :type Player%20Camera%20Manager%20Object%20Reference))
  ;[0.02248 0.1198 0.04006] AddedToCamera: Cache the multiplier.
  (|CallParentFunction Camera)
  (Variables|Default|SetCachedRotationMultiplier (ScalableFloat|GetValueatLevel (Variables|Default|GetHotfix_RotationMultiplier))))

(fn BlueprintProcessViewRotation
  (params (ViewTarget :type Actor%20Object%20Reference) (DeltaTime :type Float%20%28single-precision%29) (InViewRotation :type Rotator :default "0, 0, 0") (InDeltaRot :type Rotator :default "0, 0, 0"))
  (return InViewRotation (* InDeltaRot (Variables|Default|GetCachedRotationMultiplier)) true))
